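# Niagara idle effect for the two-handed Cranium pickaxe cosmetic that periodically flashes a sheen and sparkle on the axe head.
# AxeHead_Sheen bursts one short-lived mesh particle on randomized loops, animating a dynamic material parameter and emitting a death event.
# Axehead_Sheen_Sparkle defines a bright, offset, camera-biased sprite that fades and shrinks via curves.
# The sparkle shows no spawn module or event handler, so its link to that death event is not visible.
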NiagaraSystem /BRCosmetics/Weapons/FORT_Melee/Pickaxe_Cranium_Male/FX/NS_Pickaxe_Cranium_2h_Idle
dependencies {usedDataInterfaces [&NiagaraDataInterfaceCurve], usedDynamicInputs [], usedModules [&SystemState]}
emitters {
 AxeHead_Sheen {
  emitterUpdateScript [
   &EmitterState(
    Life Cycle Mode=::NewEnumerator1
    Loop Delay=&UniformRangedFloat(Minimum=0.5, Maximum=1)
    Loop Duration=&UniformRangedFloat(Minimum=3, Maximum=6)
    Recalculate Duration Each Loop=true
    UseLoopDelay=true
   )
   &SpawnBurst_Instantaneous(Spawn Count=1, Use Spawn Probability=false)
   &SpawnRate(Use Spawn Probability=false) {enabled=false}
  ]
  particleSpawnScript [
   /Niagara/Modules/Spawn/Initialization/InitializeParticle(
    Lifetime=0.35
    Module.Write Color=true
    Module.Write Lifetime=true
    Module.Write Mass=true
    Module.Write Material Random=true
    Module.Write Position=true
    Module.Write Scale=false
    Module.Write SpriteRotation=true
    Module.Write SpriteSize=true
    Sprite Rotation=&UniformRangedFloat(Minimum=0, Maximum=360)
    Sprite Size=&Vector2DFromFloat(Value=&UniformRangedFloat(Minimum=10, Maximum=15))
   )
   &BoxLocation(Box Offset=<0 0 -2>, Box Size=<0 0 0>, MaskSpawnGroup=false, Module.Override Seed=false, Rotate Box=false, Surface Only=false)
  ]
  particleUpdateScript [
   &ScaleColor(
    Scale Alpha=&FloatFromCurve(FloatCurve=DataInterface() {propertyValues {Curve {keys [<at 0 0>,<at 0.5016 0.9767>,<at 1 0>]}}})
    Scale RGB=&VectorFromFloat(Value=&FloatFromCurve(FloatCurve=DataInterface() {propertyValues {Curve {keys [<at 0 0>,<at 0.5017 1.0042>,<at 1 0>]}}}))
    ScaleA=true
    ScaleRGB=true
   )
   &ParticleState()
   &ScaleSpriteSize(Scale Factor=&Vector2DFromFloat(Value=&FloatFromCurve(FloatCurve=DataInterface() {propertyValues {Curve {keys [
          <at 0 0 RCIM_Cubic>
          {arriveTangent=1.9892, interpMode=RCIM_Cubic, leaveTangent=1.9892, time=0.335, value=0.2056}
          {arriveTangent=0.2473, interpMode=RCIM_Cubic, leaveTangent=0.2473, time=0.5002, value=0.995}
          {arriveTangent=-1.9908, interpMode=RCIM_Cubic, leaveTangent=-1.9908, time=0.6669, value=0.2877}
          <at 1 0>
         ]}}})))
   &DynamicMaterialParameters(
    Index 0 Param 1=&FloatFromCurve(FloatCurve=DataInterface() {propertyValues {Curve {keys [
         {arriveTangent=0.2595, interpMode=RCIM_Cubic, leaveTangent=0.2595, tangentMode=RCTM_User, time=0, value=0}
         {arriveTangent=-0.0277, interpMode=RCIM_Cubic, leaveTangent=-0.0277, tangentMode=RCTM_User, time=1.1, value=1.8}
        ]}}})
    Index 0 Param 2=1
    Index 0 Param 3=1
    Index 0 Param 4=1
    Module.Param0WriteEnabled=true
    Module.Param1WriteEnabled=true
    Module.Param2WriteEnabled=true
    Module.Param3WriteEnabled=true
   )
   &CameraOffset(Camera Offset Amount=0.05)
   &SolveForcesAndVelocity(Clamp Velocity=false, Limit Acceleration=false)
   &GenerateDeathEvent()
  ]
  rendererClasses [&NiagaraMeshRendererProperties]
  renderers [{rendererClass=&NiagaraMeshRendererProperties, rendererIndex=0}]
 }
 Axehead_Sheen_Sparkle {
  emitterUpdateScript [&EmitterState(Life Cycle Mode=::NewEnumerator1, Loop Duration=1, Recalculate Duration Each Loop=true, UseLoopDelay=true)]
  particleSpawnScript [
   &InitializeParticle(
    Color=<rgba 5 5 5 1>
    Lifetime=&UniformRangedFloat(Minimum=0.5, Maximum=0.7)
    Module.Write Color=true
    Module.Write Lifetime=true
    Module.Write Mass=true
    Module.Write Material Random=true
    Module.Write Position=true
    Module.Write Scale=false
    Module.Write SpriteRotation=true
    Module.Write SpriteSize=true
    Sprite Rotation=&UniformRangedFloat(Minimum=0, Maximum=360)
    Sprite Size=&Vector2DFromFloat(Value=&UniformRangedFloat(Minimum=27, Maximum=31))
   )
   &BoxLocation(Box Offset=<2 23 59>, Box Size=<0 0 0>, MaskSpawnGroup=false, Module.Override Seed=false, Rotate Box=false, Surface Only=false)
  ]
  particleUpdateScript [
   &ScaleColor(
    Scale Alpha=&FloatFromCurve(FloatCurve=DataInterface() {propertyValues {Curve {keys [<at 0 0>,<at 0.5016 0.9767>,<at 1 0>]}}})
    Scale RGB=&VectorFromFloat(Value=&FloatFromCurve(FloatCurve=DataInterface() {propertyValues {Curve {keys [<at 0 0>,<at 0.5017 1.0042>,<at 1 0>]}}}))
    ScaleA=true
    ScaleRGB=true
   )
   &ParticleState()
   &SolveForcesAndVelocity(Clamp Velocity=false, Limit Acceleration=false)
   &ScaleSpriteSize(Scale Factor=&Vector2DFromFloat(Value=&FloatFromCurve(FloatCurve=DataInterface() {propertyValues {Curve {keys [
          <at 0 1 RCIM_Cubic>
          {arriveTangent=-0.01, interpMode=RCIM_Cubic, leaveTangent=-0.01, time=0.335, value=1}
          {arriveTangent=-0.0228, interpMode=RCIM_Cubic, leaveTangent=-2.1462, tangentMode=RCTM_Break, time=0.5002, value=0.995}
          {arriveTangent=-1.9908, interpMode=RCIM_Cubic, leaveTangent=-1.9908, time=0.6669, value=0.2877}
          <at 1 0>
         ]}}})))
   &CameraOffset(Camera Offset Amount=20)
  ]
  rendererClasses [&NiagaraSpriteRendererProperties]
  renderers [{rendererClass=&NiagaraSpriteRendererProperties}]
 }
}
summary {
 systemName=NS_Pickaxe_Cranium_2h_Idle
 emitters [
  {emitterName=AxeHead_Sheen, rendererClasses [&NiagaraMeshRendererProperties]}
  {emitterName=Axehead_Sheen_Sparkle, rendererClasses [&NiagaraSpriteRendererProperties]}
 ]
 userVariables []
}
